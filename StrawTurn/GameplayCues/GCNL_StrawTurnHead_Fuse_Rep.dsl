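; Looping gameplay cue that shows a lit fuse on the target player's head during a timed StrawTurn duel.
; On loop start it sets replicated fuse parameters on the Hat part's idle VFX, passing the total duel time as fuse length.
; A timer plays an attached warning sound at FX_Head, timed by subtracting that sound's duration from the duel time.
; EndPlay handles timer cleanup and deactivation of the hat VFX.
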(event EventBeginPlay)

(event Collision|EventActorBeginOverlap
  (params (OtherActor :type Actor%20Object%20Reference)))

(event EventTick
  (params (DeltaSeconds :type Float%20%28single-precision%29)))

(event EventOnLoopingStart(EitherVFX-preferthisversion)
  (params (MyTarget :type Actor%20Object%20Reference) (Parameters :type Gameplay%20Cue%20Parameters%20Structure%20%28by%20ref%29) (ParticleComponents :type Array%20of%20FXSystem%20Component%20Object%20References) (AudioComponents :type Array%20of%20Audio%20Component%20Object%20References))
  (bind _targetplayerpawn (Variables|Default|GetTargetPlayerPawn))
  (bind _returnvalue (FortPawn|GetModifierforPartType _targetplayerpawn "Hat"))
  (bind _returnvalue_2 (Class|CustomCharacterPartModifier|GetIdleVFXComponent _returnvalue))
  (bind _output (Utilities|Array|Get(acopy) AudioComponents))
  ;[0 0 0] On Looping Start - Light the fuse
  (bind _asfort_player_pawn (Utilities|Casting|CastToFortPlayerPawn MyTarget))
  (Variables|Default|SetTargetPlayerPawn _asfort_player_pawn)
  (Effects|Components|ParticleSystem|SetBoolParameter _returnvalue_2 "Replicated" true)
  (Effects|Components|ParticleSystem|SetBoolParameter _returnvalue_2 "PlayFuse" true)
  (Effects|Components|ParticleSystem|SetFloatParameter _returnvalue_2 "FuseTime" (ScalableFloat|GetValueatLevel (Variables|Default|GetTotalDuelTime)))
  (Utilities|IsValid _output
    (:Is%20Valid
      (Variables|Audio|SetAmbientAudioComponent _output)
      (CallFunction|ManageWarningSound))))

(event Custom|ManageWarningSound
  (bind _totaldueltime (Variables|Default|GetTotalDuelTime))
  (bind _returnvalue (ScalableFloat|GetValueatLevel _totaldueltime))
  (bind _fusesoundwarningcue (Variables|Audio|GetFuseSoundWarningCue))
  (bind _duration (Variables|/Script/Engine.SoundBase|GetDuration _fusesoundwarningcue))
  (bind _returnvalue_2 (- _returnvalue _duration))
  ;[0 0 0] Manage Warning sound - starts timer to determine when warning sound should play
  (if (> _returnvalue_2 0)
    (bind _output_get (Variables|Default|SetTimeElapsedBeforeWarning _returnvalue_2))
    (bind _returnvalue_3 (Utilities|Time|SetTimerbyEvent (EventDispatchers|CreateEvent :!event Custom|PlayWarningSound) _output_get false true)) ; Swap sound to warning loop on fuse when the time elapses
    (Variables|Audio|SetSoundTimerHandle _returnvalue_3)))

(event Custom|PlayWarningSound
  ;[0 0 0] Manage Warning sound - starts timer to determine when warning sound should play
  ;; Play the fuse warning sound when the time runs down
  (Audio|SpawnSoundAttached :Sound (Variables|Audio|GetFuseSoundWarningCue) :AttachToComponent (Variables|/Script/Engine.Character|GetMesh (Variables|Default|GetTargetPlayerPawn)) :AttachPointName "FX_Head" :LocationType "KeepRelativeOffset" :bStopWhenAttachedToDestroyed true :VolumeMultiplier 1.0 :PitchMultiplier 1.0 :StartTime 0.0 :bAutoDestroy true))

(event Custom|FXRemoval
  (bind _targetplayerpawn (Variables|Default|GetTargetPlayerPawn))
  (bind _returnvalue (FortPawn|GetModifierforPartType _targetplayerpawn "Hat"))
  ;[0 0 0] Manage Warning sound - starts timer to determine when warning sound should play
  ;;[1 1 1] Turn off Replicated FX
  (Utilities|IsValid _targetplayerpawn
    (:Is%20Valid
      (Utilities|IsValid _returnvalue
        (:Is%20Valid
          (Components|Activation|Deactivate (Class|CustomCharacterPartModifier|GetIdleVFXComponent _returnvalue)))))))

(event EventEndPlay
  (params (EndPlayReason :type EEndPlayReason%20Enum :default "Destroyed"))
  (bind _soundtimerhandle (Variables|Audio|GetSoundTimerHandle))
  ;[0 0 0] End play - clean up timer handle and remove any lingering FX
  (|CallParentFunction EndPlayReason)
  (if (Utilities|Time|IsValidTimerHandle _soundtimerhandle)
    (Utilities|Time|ClearandInvalidateTimerbyHandle 0 _soundtimerhandle)
    (CallFunction|FXRemoval)))

(fn ConstructionScript)
